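# Body material for the medium male player character M_MED_SlowBurn, stored in that body's Demo folder.
# It uses the Default Lit shading model with legacy pixel depth offset.
# The graph holds a single vector parameter with a pale green default and shows no connections to material outputs, textures or other shading logic.
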Material /Game/Characters/Player/Male/Medium/Bodies/M_MED_SlowBurn/Demo/Body
graph {
 properties {pixelDepthOffsetMode=PDOM_Legacy, usedShadingModels=Default Lit}
 nodes [VectorParameter() {defaultValue=<rgba 0.7529 0.8422 0.7536 0>}]
}
